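# Masked parent material for the Mandala Infinity vehicle cosmetic wheel, with paintable color and glowing rim and tire.
# A shared ClockGear bump-offset function fakes infinity-mirror depth in rim and tread masks, feeding paint-tinted emissive with a sine-pulsed rim ring and tire core glow.
# Vertex color green separates rim and tire emissive, while red blends the two normal maps.
# Static switches select tire color source, ring tone and LOD alpha masking.
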Material /VehicleCosmetics/Wheels/MandalaInfinity/Materials/M_Parent_Wheel_MAndalaInfinity
graph {
 properties {
  bAutomaticallySetUsageInEditor=false
  bUsedWithMeshDeformer=true
  bUsedWithSkeletalMesh=true
  blendMode=BLEND_Masked
  previewMesh=/VehicleCosmetics/Wheels/MandalaInfinity/Meshes/SK_Wheel_MandalaInfinity
  usedShadingModels=Default Lit
 }
 nodes [
  Add()
  Ceil(None=VertexColor().G)
  Constant() {r=1}
  n1 = VectorParameter() {parameterName=TireColor, defaultValue=<rgba 0.02 0.02 0.02 1>}
  n2 = TextureSampleParameter2D() {
   parameterName=Tire_Diffuse
   samplerType=SAMPLERTYPE_LinearColor
   texture=/VehicleCosmetics/Wheels/MandalaInfinity/Texture/T_MandalaInfinity_Rim_DiffSpec_Infinity
  }
  n3 = VectorParameter() {parameterName=PaintColor, defaultValue=<rgba 0.72 0.33 1 1>}
  n4 = ScalarParameter() {defaultValue=1, parameterName=Paint Boost}
  n5 = TextureSampleParameter2D() {
   parameterName=Tire_Masks
   samplerType=SAMPLERTYPE_LinearColor
   texture=/VehicleCosmetics/Wheels/MandalaInfinity/Texture/T_MandalaInfinity_Tire_RGB_SE
  }
  n6 = Add(A=Add(A=@n5.G, B=@n5.G), B=Multiply(A=@n5.G, B=ScalarParameter() {defaultValue=20, parameterName=PaintBoostWithoutAffectingEmissive_Tire}))
  LinearInterpolate(
   A=Multiply(A=StaticSwitchParameter(True=@n1.RGB, False=@n2.RGB) {defaultValue=true, parameterName=UseTireColor}, B=Constant() {r=16})
   B=@n3.RGB
   Alpha=Clamp(None=Multiply(A=@n4, B=@n6))
  )
  n7 = Constant()
  n8 = TextureObjectParameter() {
   parameterName=Mask_Offset
   samplerType=SAMPLERTYPE_LinearColor
   texture=/VehicleCosmetics/Wheels/MandalaInfinity/Texture/T_MandalaInfinity_Rim_Inf_Mask
  }
  n9 = ComponentMask(None=MaterialFunctionCall!(
    Texture=@n8
    Offset=ScalarParameter() {defaultValue=-0.05, parameterName=BumpOffsetAmount_RmInf}
    UVs
    RotateUVs
    GradientFalloff
    GradientDarkness=ScalarParameter() {defaultValue=0.0008, parameterName=GradientDarkness_RimInf}
   ) {materialFunction=/VehicleCosmetics/Wheels/ClockGear_SE/Materials/MF_Infinity_OffsetFunction}.Result) {g=true}
  n10 = Multiply(A=Multiply(A=@n9, B=@n9), B=ScalarParameter() {defaultValue=20, parameterName=PaintBoostWithoutAffectingEmissive_Rim})
  n11 = TextureSampleParameter2D() {parameterName=RimMask2, samplerType=SAMPLERTYPE_LinearColor, texture=/VehicleCosmetics/Wheels/MandalaInfinity/Texture/Mandala_Rim_RGB2}
  n12 = Add(A=@n11.R, B=@n11.R)
  n13 = TextureSampleParameter2D() {parameterName=Rim_Masks, samplerType=SAMPLERTYPE_LinearColor, texture=/VehicleCosmetics/Wheels/MandalaInfinity/Texture/Mandala_Rim_RGB}
  n14 = StaticSwitchParameter(True=@n3.RGB, False=VectorParameter() {parameterName=RingEmissiveCOlor, defaultValue=<rgba 0.25 0.25 0.25 1>}.RGB) {defaultValue=true, parameterName=PaintTone or DuoTone}
  n15 = Add(A=LinearInterpolate(A=@n7, B=@n3.RGB, Alpha=Add(A=@n10, B=@n12)), B=Multiply(A=@n13.R, B=@n14))
  n16 = Sine(None=Multiply(A=Time(), B=Constant() {r=0.333}))
  n17 = Clamp(None=Multiply(A=Multiply(A=@n16, B=@n13.R), B=@n14))
  n18 = Power(Base=@n5.A, Exp=ScalarParameter() {defaultValue=1.5, parameterName=CoreGlowBack_PowerExponent})
  n19 = Multiply(A=Multiply(A=@n18, B=@n14), B=ScalarParameter() {defaultValue=0.78, parameterName=CoreGlowBack_Multiplier})
  n20 = TextureObject() {samplerType=SAMPLERTYPE_LinearColor, texture=/VehicleCosmetics/Wheels/MandalaInfinity/Texture/T_MandalaInfinity_Tire_inf_Mask_spaced}
  n21 = ComponentMask(None=MaterialFunctionCall(
    Texture=@n20
    Offset=ScalarParameter() {defaultValue=-0.05, parameterName=BumpOffsetAmount_TreadInf}
    GradientDarkness=ScalarParameter() {defaultValue=0.0008, parameterName=GradientDarkness_TreadInf}
   ) {materialFunction=&MF_Infinity_OffsetFunction}.Result) {r=true}
  n22 = Add(A=Add(A=@n21, B=@n21), B=@n18)
  n23 = LinearInterpolate(A=@n7, B=@n3.RGB, Alpha=Add(A=@n22, B=@n22))
  LinearInterpolate(
   A=Multiply(A=Add(A=@n15, B=@n17), B=ScalarParameter() {defaultValue=2, parameterName=RimEmissive})
   B=Multiply(A=Add(A=@n19, B=@n23), B=ScalarParameter() {defaultValue=10, parameterName=TireEmissive})
   Alpha=Ceil(None=VertexColor().G)
  )
  n24 = TextureSampleParameter2D() {parameterName=Rim_N, samplerType=SAMPLERTYPE_Normal, texture=/VehicleCosmetics/Wheels/MandalaInfinity/Texture/T_MandalaInfinity_Tire_N}
  n25 = TextureSampleParameter2D() {parameterName=Tier_N, samplerType=SAMPLERTYPE_Normal, texture=/VehicleCosmetics/Wheels/MandalaInfinity/Texture/T_MandalaInfinity_Rim_N}
  LinearInterpolate(A=@n24.RGB, B=@n25.RGB, Alpha=VertexColor().R)
  LinearInterpolate(A=Constant(), B=@n13.B)
  Multiply(A=@n11.B)
  Multiply(A=@n13.A)
  Reroute(None=@n13.B)
  ScalarParameter() {defaultValue=1, parameterName=Roughness}
  ScalarParameter() {parameterName=Metallic}
  n26 = VertexColor()
  n27 = TextureSampleParameter2D() {
   parameterName=AlphaLOD
   samplerType=SAMPLERTYPE_LinearColor
   texture=/VehicleCosmetics/Wheels/MandalaInfinity/Texture/T_MandalaInfinity_Rim_Alpha_LOD
  }
  StaticSwitchParameter(True=Constant() {r=1}, False=LinearInterpolate(A=Add(A=@n26.R, B=@n26.R), B=Constant() {r=1}, Alpha=@n27.R)) {parameterName=LODs_Alpha}
 ]
}
